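# Niagara system for the RaveVision cosmetic's skydiving dive contrail, trailing small mesh particles from the character's body and hands.
# Active crackle emitters sample bones through a shared skeletal mesh reader, using spawn groups to split bones and offsets.
# World-up wind, fading noise and drag stream particles away, with larger pieces living shorter.
# User.VariantColor tints them, front-end modifiers adjust size and drag, and several sprite and glitch emitters are disabled.
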NiagaraSystem /BRCosmetics/Athena/Effects/Particles/Skydiving/NS_Contrail_RaveVision/NS_Contrail_Dive_RaveVision
dependencies {
 usedDataInterfaces [
  &NiagaraDataInterfaceSkeletalMesh
  &NiagaraDataInterfaceCurve
  &NiagaraDataInterfaceArrayInt32
  &NiagaraDataInterfaceArrayFloat
  &NiagaraDataInterfaceMeshRendererInfo
 ]
 usedDynamicInputs [&SetBoolByFloatComparison, &FN_FrontEndVector3Modifier]
 usedModules [
  &self:SystemSpawnScript.SystemScriptSource.SystemScriptGraph.NiagaraNodeAssignment_2.SetVariables_#1
  &SystemState
  /Game/Effects/Niagara/Modules/Skydiving/FN_DivingAttributes/v5/FN_DivingAttributes
  &FN_Integrator
  &self:SystemSpawnScript.SystemScriptSource.SystemScriptGraph.NiagaraNodeAssignment_0.SetVariables_#2
  &self:SystemSpawnScript.SystemScriptSource.SystemScriptGraph.NiagaraNodeAssignment_1.SetVariables_#3
 ]
}
emitters {
 BodyCrackle {
  bEnabled=false
  emitterUpdateScript [
   &EmitterState()
   &SpawnRate(SpawnRate=70, Use Spawn Probability=false)
  ]
  particleSpawnScript [
   &FN_SetSpawnGroup(Number of Spawn Groups=2)
   &InitializeParticle(
    Color=&Multiply_LinearColorByFloat(Linear Color=$User.VariantColor, Float=0.5)
    Lifetime Max=0.45
    Lifetime Min=0.3
    Lifetime Mode=::NewEnumerator1
    Mesh Scale Mode=::NewEnumerator3
    Sprite Size Mode=::NewEnumerator3
    Uniform Sprite Size=&Lerp_Float(
     A=2
     B=4
     Alpha=&FloatFromCurve(
      FloatCurve=DataInterface() {propertyValues {Curve {keys [
          {interpMode=RCIM_Cubic, tangentMode=RCTM_User, time=0, value=0}
          {arriveTangent=2.3987, interpMode=RCIM_Cubic, leaveTangent=2.3987, tangentMode=RCTM_User, time=0.6, value=0.55}
          {interpMode=RCIM_Cubic, tangentMode=RCTM_User, time=1, value=1}
         ]}}}
      CurveIndex=&RandomRangeFloat(Minimum=0, Maximum=1)
     )
    )
    UsePositionOffset=false
    Write Material Random=true
   )
   &MeshRotationForce(Rotation=&RandomRangeVector(Minimum=<5 5 5>, Maximum=<-5 -5 -5>))
   &ApplyInitialForces()
   &SkeletalMeshLocation(
    Bone / Socket Index=&MaskIntBySpawnGroup(Masked Int=5, Passthrough Int=6, Spawn Group Mask=0)
    Bone Sampling Mode=::NewEnumerator4
    Interpolate Mesh Sampling=false
    MaskSpawnGroup=false
    Orientation Sampling=::NewEnumerator1
    Position Sampling=::NewEnumerator1
    Sampled Position Offset=&MaskVectorBySpawnGroup(
     Masked Vector=&RandomRangeVector(Minimum=<-3 0 0>, Maximum=<25 0 0>)
     Passthrough Vector=&RandomRangeVector(Minimum=<-25 0 0>, Maximum=<3 0 0>)
    )
   )
   &SphereLocation(
    MaskSpawnGroup=false
    Override Seed=false
    Sphere Radius=2
    Surface Only=false
    UseNonUniformScale=false
    UseOffset=false
    UseSphereOrientation=false
   )
   &CameraOffset(Camera Offset Amount=3)
  ]
  particleUpdateScript [
   &ParticleState()
   &VectorNoiseForce(
    Force Amount=&FloatFromCurve(
     FloatCurve=DataInterface() {propertyValues {Curve {keys [{interpMode=RCIM_Cubic, tangentMode=RCTM_User, time=0, value=1}, {interpMode=RCIM_Cubic, tangentMode=RCTM_User, time=1, value=0}]}}}
     Scale Curve=700
    )
    Module.Use Random Noise Time Interval=false
    Random Noise Time Interval=0.15
   )
   &WindForce(
    Wind Speed=&TransformVector(Vector=<0 0 1>, Source Space=::World, Destination Space=::Local)
    Wind Speed Scale=800
   )
   &Drag(Drag=7, Use Linear Drag=true, Use Rotational Drag=false)
   &ScaleSpriteSize(
    Uniform Curve Scale=&Waveform(
     Global Phase=$Particles.Age
     "Waveform [1]"=::NewEnumerator12
     "[1] Frequency"=&RandomRangeFloat(Minimum=15, Maximum=40)
     "[1] Amplitude Min/Max"=<0.5 1.5>
    )
    Uniform Curve Sprite Scale=DataInterface() {propertyValues {Curve {keys [
        {arriveTangent=10.8984, interpMode=RCIM_Cubic, leaveTangent=10.8984, tangentMode=RCTM_User, time=0, value=0}
        <at 0.2 1>
        {interpMode=RCIM_Cubic, tangentMode=RCTM_User, time=0.6, value=1}
        {arriveTangent=-5.0292, interpMode=RCIM_Cubic, leaveTangent=-5.0292, tangentMode=RCTM_User, time=1, value=0}
       ]}}}
   )
   /SetVariables(
    Particles.SpriteParams0:Vector4f=&MakeFloat4FromVectorAndFloat("Vector (RGB)"=&MakeVectorFromLinearColorRGB(Color=<rgba 0 0.8192 1 1>), "Float (Alpha)"=1)
    Particles.SpriteParams1:Vector4f=&MakeFloat4FromVectorAndFloat(
     "Vector (RGB)"=&Subtract_Vector(A=$Particles.SkeletalMeshLocation.SampledPosition, B=Linked() {linkedVariable {name=Particles.Position, type=NiagaraPosition}})
     "Float (Alpha)"=2
    )
   ) {isSetParameters=true, module=SetVariables_#4}
   &SolveForcesAndVelocity(Clamp Velocity=false, Limit Acceleration=false)
  ]
  rendererClasses [&NiagaraSpriteRendererProperties]
  renderers [{rendererClass=&NiagaraSpriteRendererProperties}]
 }
 BodyCrackle001 {
  emitterUpdateScript [
   &EmitterState()
   &SpawnRate(SpawnRate=125, Use Spawn Probability=false)
  ]
  particleSpawnScript [
   &FN_SetSpawnGroup(Number of Spawn Groups=2)
   &InitializeParticle(
    Color=&Multiply_LinearColorByFloat(Linear Color=$User.VariantColor, Float=2)
    Lifetime Max=0.55
    Lifetime Min=0.25
    Lifetime Mode=::NewEnumerator1
    Mesh Renderer Array Visibility Mode=::NewEnumerator3
    Mesh Renderer Info=DataInterface() {propertyValues {MeshRenderer=&self:BodyCrackle001_0.NiagaraMeshRendererProperties_0}}
    Mesh Scale Mode=::NewEnumerator3
    Mesh Uniform Scale=&Multiply_Float(
     A=&RemapRange(
      Input Value To Remap=&FloatFromCurve(
       FloatCurve=DataInterface() {propertyValues {Curve {
          defaultValue=0.05
          keys [
           {tangentMode=RCTM_User, time=0, value=0}
           {
            arriveTangent=0.0405
            arriveTangentWeight=0.2338
            interpMode=RCIM_Cubic
            leaveTangent=0.0405
            leaveTangentWeight=0.0998
            tangentMode=RCTM_User
            time=0.77
            value=0
           }
           {arriveTangent=12.0972, arriveTangentWeight=0.8048, interpMode=RCIM_Cubic, leaveTangent=12.0972, tangentMode=RCTM_User, time=1, value=1}
          ]
         }}}
       CurveIndex=&RandomRangeFloat(Minimum=0, Maximum=1)
      )
      Input Min=0
      Input Max=1
      Output Min=0.05
      Output Max=0.15
     )
     B=&FN_FrontEndFloatModifier(Gameplay=1.5, Front End=1.5)
    )
    UsePositionOffset=false
    Write Material Random=true
   )
   /SetVariables(Particles.Lifetime=&Multiply_Float(
     A=$Particles.Lifetime
     B=&RemapRange(
      Input Value To Remap=&VectorLength(Vector=$Particles.Scale)
      Input Min=&FN_FrontEndFloatModifier(Gameplay=0.07, Front End=0.15)
      Input Max=&FN_FrontEndFloatModifier(Gameplay=0.18, Front End=0.25)
      Output Min=1.22
      Output Max=0.35
     )
    )) {isSetParameters=true}
   &MeshRotationForce(Rotation=&RandomRangeVector(Minimum=<5 5 5>, Maximum=<-5 -5 -5>))
   &ApplyInitialForces()
   &SkeletalMeshLocation(
    Bone / Socket Index=&MaskIntBySpawnGroup(Masked Int=5, Passthrough Int=6, Spawn Group Mask=0)
    Bone Sampling Mode=::NewEnumerator4
    Interpolate Mesh Sampling=false
    MaskSpawnGroup=false
    Orientation Sampling=::NewEnumerator1
    Position Sampling=::NewEnumerator1
    Sampled Position Offset=&MaskVectorBySpawnGroup(
     Masked Vector=&RandomRangeVector(Minimum=<10 0 0>, Maximum=<25 0 0>)
     Passthrough Vector=&RandomRangeVector(Minimum=<-25 0 0>, Maximum=<-10 0 0>)
    )
    Skeletal Mesh=$System.Skeletal Mesh Reader
   )
   &SphereLocation(
    MaskSpawnGroup=false
    Override Seed=false
    Sphere Radius=2
    Surface Only=false
    UseNonUniformScale=false
    UseOffset=false
    UseSphereOrientation=false
   )
   &CameraOffset(Camera Offset Amount=3)
  ]
  particleUpdateScript [
   &ParticleState()
   &VectorNoiseForce(
    Force Amount=&FloatFromCurve(
     FloatCurve=DataInterface() {propertyValues {Curve {keys [{interpMode=RCIM_Cubic, tangentMode=RCTM_User, time=0, value=1}, {interpMode=RCIM_Cubic, tangentMode=RCTM_User, time=1, value=0}]}}}
     Scale Curve=700
    )
    Module.Use Random Noise Time Interval=false
    Random Noise Time Interval=0.15
   )
   &WindForce(
    Wind Speed=&TransformVector(Vector=<0 0 1>, Source Space=::World, Destination Space=::Local)
    Wind Speed Scale=800
   )
   &Drag(
    Drag=&Multiply_Float(
     A=&FN_FrontEndFloatModifier(Gameplay=6, Front End=7)
     B=&RemapRange(
      Input Value To Remap=&VectorLength(Vector=$Particles.Scale)
      Input Min=0.05
      Input Max=1
      Output Min=1
      Output Max=1.5
     )
    )
    Use Linear Drag=true
    Use Rotational Drag=false
   )
   &ScaleMeshSize(Scale Factor=&VectorFromFloat(Value=&FloatFromCurve(FloatCurve=DataInterface() {propertyValues {Curve {
         defaultValue=1
         keys [
          <at 0 0 RCIM_Cubic>
          {arriveTangent=1.4268, leaveTangent=1.4268, time=0.0935, value=1}
          {arriveTangent=-0.0405, interpMode=RCIM_Cubic, leaveTangent=-0.0405, tangentMode=RCTM_User, time=0.7009, value=1}
          {arriveTangent=-8.0089, interpMode=RCIM_Cubic, leaveTangent=-8.0089, tangentMode=RCTM_User, time=1, value=0}
         ]
        }}})))
   &SolveForcesAndVelocity(Clamp Velocity=false, Limit Acceleration=false)
  ]
  rendererClasses [&NiagaraMeshRendererProperties]
  renderers [{rendererClass=&NiagaraMeshRendererProperties, rendererIndex=0}]
 }
 BodyCrackle002 {
  bEnabled=false
  emitterUpdateScript [
   &EmitterState()
   &SpawnRate(SpawnRate=70, Use Spawn Probability=false)
  ]
  particleSpawnScript [
   &FN_SetSpawnGroup(Number of Spawn Groups=2)
   &InitializeParticle(
    Color=&Multiply_LinearColorByFloat(Linear Color=$User.VariantColor, Float=0.5)
    Lifetime Max=0.45
    Lifetime Min=0.3
    Lifetime Mode=::NewEnumerator1
    Mesh Scale Mode=::NewEnumerator3
    Sprite Size Mode=::NewEnumerator3
    Uniform Sprite Size=&Lerp_Float(
     A=2
     B=4
     Alpha=&FloatFromCurve(
      FloatCurve=DataInterface() {propertyValues {Curve {keys [
          {interpMode=RCIM_Cubic, tangentMode=RCTM_User, time=0, value=0}
          {arriveTangent=2.3987, interpMode=RCIM_Cubic, leaveTangent=2.3987, tangentMode=RCTM_User, time=0.6, value=0.55}
          {interpMode=RCIM_Cubic, tangentMode=RCTM_User, time=1, value=1}
         ]}}}
      CurveIndex=&RandomRangeFloat(Minimum=0, Maximum=1)
     )
    )
    UsePositionOffset=false
    Write Material Random=true
   )
   &MeshRotationForce(Rotation=&RandomRangeVector(Minimum=<5 5 5>, Maximum=<-5 -5 -5>))
   &ApplyInitialForces()
   &SkeletalMeshLocation(
    Bone / Socket Index=&MaskIntBySpawnGroup(Masked Int=5, Passthrough Int=6, Spawn Group Mask=0)
    Bone Sampling Mode=::NewEnumerator4
    Interpolate Mesh Sampling=false
    MaskSpawnGroup=false
    Orientation Sampling=::NewEnumerator1
    Position Sampling=::NewEnumerator1
    Sampled Position Offset=&MaskVectorBySpawnGroup(
     Masked Vector=&RandomRangeVector(Minimum=<-3 0 0>, Maximum=<25 0 0>)
     Passthrough Vector=&RandomRangeVector(Minimum=<-25 0 0>, Maximum=<3 0 0>)
    )
   )
   &SphereLocation(
    MaskSpawnGroup=false
    Override Seed=false
    Sphere Radius=2
    Surface Only=false
    UseNonUniformScale=false
    UseOffset=false
    UseSphereOrientation=false
   )
   &CameraOffset(Camera Offset Amount=3)
  ]
  particleUpdateScript [
   &ParticleState()
   &VectorNoiseForce(
    Force Amount=&FloatFromCurve(
     FloatCurve=DataInterface() {propertyValues {Curve {keys [{interpMode=RCIM_Cubic, tangentMode=RCTM_User, time=0, value=1}, {interpMode=RCIM_Cubic, tangentMode=RCTM_User, time=1, value=0}]}}}
     Scale Curve=700
    )
    Module.Use Random Noise Time Interval=false
    Random Noise Time Interval=0.15
   )
   &WindForce(
    Wind Speed=&TransformVector(Vector=<0 0 1>, Source Space=::World, Destination Space=::Local)
    Wind Speed Scale=800
   )
   &Drag(Drag=7, Use Linear Drag=true, Use Rotational Drag=false)
   &ScaleSpriteSize(
    Uniform Curve Scale=&Waveform(
     Global Phase=$Particles.Age
     "Waveform [1]"=::NewEnumerator12
     "[1] Frequency"=&RandomRangeFloat(Minimum=15, Maximum=40)
     "[1] Amplitude Min/Max"=<0.5 1.5>
    )
    Uniform Curve Sprite Scale=DataInterface() {propertyValues {Curve {keys [
        {arriveTangent=10.8984, interpMode=RCIM_Cubic, leaveTangent=10.8984, tangentMode=RCTM_User, time=0, value=0}
        <at 0.2 1>
        {interpMode=RCIM_Cubic, tangentMode=RCTM_User, time=0.6, value=1}
        {arriveTangent=-5.0292, interpMode=RCIM_Cubic, leaveTangent=-5.0292, tangentMode=RCTM_User, time=1, value=0}
       ]}}}
   )
   /SetVariables(
    Particles.SpriteParams0:Vector4f=&MakeFloat4FromVectorAndFloat("Vector (RGB)"=&MakeVectorFromLinearColorRGB(Color=<rgba 0 0.8192 1 1>), "Float (Alpha)"=1)
    Particles.SpriteParams1:Vector4f=&MakeFloat4FromVectorAndFloat(
     "Vector (RGB)"=&Subtract_Vector(A=$Particles.SkeletalMeshLocation.SampledPosition, B=Linked() {linkedVariable {name=Particles.Position, type=NiagaraPosition}})
     "Float (Alpha)"=2
    )
   ) {isSetParameters=true, module=SetVariables_#4}
   &SolveForcesAndVelocity(Clamp Velocity=false, Limit Acceleration=false)
  ]
  rendererClasses [&NiagaraSpriteRendererProperties]
  renderers [{rendererClass=&NiagaraSpriteRendererProperties}]
 }
 BodyCrackle003 {
  emitterUpdateScript [
   &EmitterState()
   &SpawnRate(SpawnRate=100, Use Spawn Probability=false)
  ]
  particleSpawnScript [
   &FN_SetSpawnGroup(Number of Spawn Groups=2)
   &InitializeParticle(
    Color=&Multiply_LinearColorByFloat(Linear Color=$User.VariantColor, Float=2)
    Lifetime Max=0.55
    Lifetime Min=0.25
    Lifetime Mode=::NewEnumerator1
    Mesh Renderer Array Visibility Mode=::NewEnumerator3
    Mesh Renderer Info=DataInterface() {propertyValues {MeshRenderer=&self:BodyCrackle003_1.NiagaraMeshRendererProperties_0}}
    Mesh Scale Mode=::NewEnumerator3
    Mesh Uniform Scale=&Multiply_Float(
     A=&RemapRange(
      Input Value To Remap=&FloatFromCurve(
       FloatCurve=DataInterface() {propertyValues {Curve {
          defaultValue=0.05
          keys [
           {tangentMode=RCTM_User, time=0, value=0}
           {
            arriveTangent=0.0405
            arriveTangentWeight=0.2338
            interpMode=RCIM_Cubic
            leaveTangent=0.0405
            leaveTangentWeight=0.0998
            tangentMode=RCTM_User
            time=0.77
            value=0
           }
           {arriveTangent=12.0972, arriveTangentWeight=0.8048, interpMode=RCIM_Cubic, leaveTangent=12.0972, tangentMode=RCTM_User, time=1, value=1}
          ]
         }}}
       CurveIndex=&RandomRangeFloat(Minimum=0, Maximum=1)
      )
      Input Min=0
      Input Max=1
      Output Min=&FN_FrontEndFloatModifier(Gameplay=0.0375, Front End=0.05)
      Output Max=0.15
     )
     B=&FN_FrontEndFloatModifier(Gameplay=1.5, Front End=1.5)
    )
    UsePositionOffset=false
    Write Material Random=true
   )
   /SetVariables(Particles.Lifetime=&Multiply_Float(
     A=$Particles.Lifetime
     B=&RemapRange(
      Input Value To Remap=&VectorLength(Vector=$Particles.Scale)
      Input Min=&FN_FrontEndFloatModifier(Gameplay=0.07, Front End=0.15)
      Input Max=&FN_FrontEndFloatModifier(Gameplay=0.18, Front End=0.25)
      Output Min=1.22
      Output Max=0.15
     )
    )) {isSetParameters=true}
   &MeshRotationForce(Rotation=&RandomRangeVector(Minimum=<5 5 5>, Maximum=<-5 -5 -5>))
   &ApplyInitialForces()
   &SkeletalMeshLocation(
    Bone / Socket Index=&MaskIntBySpawnGroup(Masked Int=2, Passthrough Int=3, Spawn Group Mask=0)
    Bone Sampling Mode=::NewEnumerator4
    Interpolate Mesh Sampling=false
    MaskSpawnGroup=false
    Orientation Sampling=::NewEnumerator1
    Position Sampling=::NewEnumerator1
    Sampled Position Offset=&MaskVectorBySpawnGroup(
     Masked Vector=&RandomRangeVector(Minimum=<0 -15 0>, Maximum=<5 5 0>)
     Passthrough Vector=&RandomRangeVector(Minimum=<0 15 0>, Maximum=<-5 -5 0>)
    )
    Skeletal Mesh=$System.Skeletal Mesh Reader
   )
   &SphereLocation(
    MaskSpawnGroup=false
    Override Seed=false
    Sphere Radius=2
    Surface Only=false
    UseNonUniformScale=false
    UseOffset=false
    UseSphereOrientation=false
   )
   &CameraOffset(Camera Offset Amount=3)
  ]
  particleUpdateScript [
   &ParticleState()
   &VectorNoiseForce(
    Force Amount=&FloatFromCurve(
     FloatCurve=DataInterface() {propertyValues {Curve {keys [{interpMode=RCIM_Cubic, tangentMode=RCTM_User, time=0, value=1}, {interpMode=RCIM_Cubic, tangentMode=RCTM_User, time=1, value=0}]}}}
     Scale Curve=700
    )
    Module.Use Random Noise Time Interval=false
    Random Noise Time Interval=0.15
   )
   &WindForce(
    Wind Speed=&TransformVector(Vector=<0 0 1>, Source Space=::World, Destination Space=::Local)
    Wind Speed Scale=800
   )
   &Drag(
    Drag=&Multiply_Float(
     A=&FN_FrontEndFloatModifier(Gameplay=6, Front End=7)
     B=&RemapRange(
      Input Value To Remap=&VectorLength(Vector=$Particles.Scale)
      Input Min=0.05
      Input Max=1
      Output Min=1
      Output Max=1.5
     )
    )
    Use Linear Drag=true
    Use Rotational Drag=false
   )
   &ScaleMeshSize(Scale Factor=&VectorFromFloat(Value=&FloatFromCurve(FloatCurve=DataInterface() {propertyValues {Curve {
         defaultValue=1
         keys [
          <at 0 0 RCIM_Cubic>
          {arriveTangent=1.4268, leaveTangent=1.4268, time=0.0935, value=1}
          {arriveTangent=-0.0405, interpMode=RCIM_Cubic, leaveTangent=-0.0405, tangentMode=RCTM_User, time=0.7009, value=1}
          {arriveTangent=-8.0089, interpMode=RCIM_Cubic, leaveTangent=-8.0089, tangentMode=RCTM_User, time=1, value=0}
         ]
        }}})))
   &PointAttractionForce(
    Attraction Radius=1498.6685
    Attraction Strength=58.4253
    Attractor Position Offset:Vector3f=&AddVectorToPosition(Position=$Particles.SkeletalMeshLocation.SampledPosition, Vector=<-15 -15 15>)
    Attractor Position Offset Coordinate Space:ENiagaraCoordinateSpace=::Local
    Kill Within Radius=false
    Module.Use Falloff=true
    UseAttractorOffset=true
   )
   &SolveForcesAndVelocity(Clamp Velocity=false, Limit Acceleration=false)
  ]
  rendererClasses [&NiagaraMeshRendererProperties]
  renderers [{rendererClass=&NiagaraMeshRendererProperties, rendererIndex=0}]
 }
 BodyCrackle004 {
  bEnabled=false
  emitterUpdateScript [
   &EmitterState()
   &SpawnRate(SpawnRate=15, Use Spawn Probability=false)
  ]
  particleSpawnScript [
   &FN_SetSpawnGroup(Number of Spawn Groups=2)
   &InitializeParticle(
    Color=&Multiply_LinearColorByFloat(Linear Color=$User.VariantColor, Float=2)
    Lifetime=0.48
    Mesh Renderer Array Visibility Mode=::NewEnumerator3
    Mesh Renderer Info=DataInterface() {propertyValues {MeshRenderer=&self:BodyCrackle004_0.NiagaraMeshRendererProperties_0}}
    Mesh Scale Mode=::NewEnumerator3
    Mesh Uniform Scale=0.25
    UsePositionOffset=false
    Write Material Random=true
   )
   &MeshRotationForce(Rotation=&RandomRangeVector(Minimum=<10 10 10>, Maximum=<-10 -10 -10>))
   &ApplyInitialForces()
   &SkeletalMeshLocation(
    Bone / Socket Index=&MaskIntBySpawnGroup(Masked Int=0, Passthrough Int=1, Spawn Group Mask=0)
    Bone Sampling Mode=::NewEnumerator4
    Interpolate Mesh Sampling=false
    MaskSpawnGroup=false
    Orientation Sampling=::NewEnumerator1
    Position Sampling=::NewEnumerator1
    Sampled Position Offset=&MaskVectorBySpawnGroup(Masked Vector=<8 0 0>, Passthrough Vector=<-8 0 0>)
   )
   &SphereLocation(
    MaskSpawnGroup=false
    Override Seed=false
    Sphere Radius=2
    Surface Only=false
    UseNonUniformScale=false
    UseOffset=false
    UseSphereOrientation=false
   )
   &CameraOffset(Camera Offset Amount=3)
  ]
  particleUpdateScript [
   &ParticleState()
   &VectorNoiseForce(
    Force Amount=&FloatFromCurve(
     FloatCurve=DataInterface() {propertyValues {Curve {keys [{interpMode=RCIM_Cubic, tangentMode=RCTM_User, time=0, value=1}, {interpMode=RCIM_Cubic, tangentMode=RCTM_User, time=1, value=0}]}}}
     Scale Curve=700
    )
    Module.Use Random Noise Time Interval=false
    Random Noise Time Interval=0.15
   )
   &WindForce(
    Wind Speed=&TransformVector(
     Vector=&RandomRangeVector(Minimum=<-0.25 -0.25 1>, Maximum=<0.25 0.25 1>)
     Source Space=::World
     Destination Space=::Local
    )
    Wind Speed Scale=&FloatFromCurve(
     FloatCurve=DataInterface() {propertyValues {Curve {
        defaultValue=1
        keys [
         {arriveTangent=1.3406, interpMode=RCIM_Cubic, leaveTangent=1.3406, tangentMode=RCTM_User, time=0, value=0}
         {
          arriveTangent=3.3934
          arriveTangentWeight=0.8294
          interpMode=RCIM_Cubic
          leaveTangent=3.3934
          leaveTangentWeight=0.8294
          tangentMode=RCTM_User
          tangentWeightMode=RCTWM_WeightedBoth
          time=0.6
          value=1
         }
        ]
       }}}
     Scale Curve=1200
    )
   )
   &Drag(Drag=6, Use Linear Drag=true, Use Rotational Drag=false)
   &ScaleMeshSize(Scale Factor=&VectorFromFloat(Value=&FloatFromCurve(FloatCurve=DataInterface() {propertyValues {Curve {
         defaultValue=1
         keys [
          <at 0 0 RCIM_Cubic>
          {arriveTangent=1.4268, leaveTangent=1.4268, time=0.0935, value=1}
          {arriveTangent=-0.0405, interpMode=RCIM_Cubic, leaveTangent=-0.0405, tangentMode=RCTM_User, time=0.3, value=1}
          {arriveTangent=-4.4328, interpMode=RCIM_Cubic, leaveTangent=-4.4328, tangentMode=RCTM_User, time=1, value=0}
         ]
        }}})))
   &SolveForcesAndVelocity(Clamp Velocity=false, Limit Acceleration=false)
  ]
  rendererClasses [&NiagaraMeshRendererProperties]
  renderers [{rendererClass=&NiagaraMeshRendererProperties, rendererIndex=0}]
 }
 BodyCrackle005 {
  emitterUpdateScript [
   &EmitterState()
   &SpawnRate(SpawnRate=20, Use Spawn Probability=false)
  ]
  particleSpawnScript [
   &FN_SetSpawnGroup()
   &InitializeParticle(
    Color=&Multiply_LinearColorByFloat(Linear Color=$User.VariantColor, Float=2)
    Lifetime=0.48
    Mesh Renderer Array Visibility Mode=::NewEnumerator3
    Mesh Renderer Info=DataInterface() {propertyValues {MeshRenderer=&self:BodyCrackle005_0.NiagaraMeshRendererProperties_0}}
    Mesh Scale Mode=::NewEnumerator3
    Mesh Uniform Scale=&MergeHelper_FloatSwitch(
     0=&FN_FrontEndFloatModifier(Gameplay=0.18, Front End=0.25)
     1=&FN_FrontEndFloatModifier(Gameplay=0.18, Front End=0.25)
     2=0.25
     3=0.25
     Index=&SpawnGroup()
     Num=::NewEnumerator2
    )
    UsePositionOffset=false
    Write Material Random=true
   )
   &MeshRotationForce(Rotation=&RandomRangeVector(Minimum=<10 10 10>, Maximum=<-10 -10 -10>))
   &ApplyInitialForces()
   &SkeletalMeshLocation(
    Bone / Socket Index=&MergeHelper_IntSwitch(0=2, 1=3, 2=0, 3=1, 4=?, 5=?, Index=&SpawnGroup(), Num=::NewEnumerator2)
    Bone Sampling Mode=::NewEnumerator4
    Interpolate Mesh Sampling=false
    MaskSpawnGroup=false
    Orientation Sampling=::NewEnumerator1
    Position Sampling=::NewEnumerator1
    Sampled Position Offset=&MergeHelper_VectorSwitch(0=<0 5 0>, 1=<0 -5 0>, 2=<8 0 0>, 3=<-8 0 0>, Index=&SpawnGroup(), Num=::NewEnumerator2)
    Skeletal Mesh=$System.Skeletal Mesh Reader
   )
   &SphereLocation(
    MaskSpawnGroup=false
    Override Seed=false
    Sphere Radius=2
    Surface Only=false
    UseNonUniformScale=false
    UseOffset=false
    UseSphereOrientation=false
   )
   &CameraOffset(Camera Offset Amount=3)
  ]
  particleUpdateScript [
   &ParticleState()
   &VectorNoiseForce(
    Force Amount=&FloatFromCurve(
     FloatCurve=DataInterface() {propertyValues {Curve {keys [{interpMode=RCIM_Cubic, tangentMode=RCTM_User, time=0, value=1}, {interpMode=RCIM_Cubic, tangentMode=RCTM_User, time=1, value=0}]}}}
     Scale Curve=700
    )
    Module.Use Random Noise Time Interval=false
    Random Noise Time Interval=0.15
   )
   &WindForce(
    Wind Speed=&TransformVector(
     Vector=&RandomRangeVector(Minimum=<-0.25 -0.25 1>, Maximum=<0.25 0.25 1>)
     Source Space=::World
     Destination Space=::Local
    )
    Wind Speed Scale=&FloatFromCurve(
     FloatCurve=DataInterface() {propertyValues {Curve {
        defaultValue=1
        keys [
         {arriveTangent=1.3406, interpMode=RCIM_Cubic, leaveTangent=1.3406, tangentMode=RCTM_User, time=0, value=0}
         {
          arriveTangent=3.3934
          arriveTangentWeight=0.8294
          interpMode=RCIM_Cubic
          leaveTangent=3.3934
          leaveTangentWeight=0.8294
          tangentMode=RCTM_User
          tangentWeightMode=RCTWM_WeightedBoth
          time=0.6
          value=1
         }
        ]
       }}}
     Scale Curve=1200
    )
   )
   &Drag(Drag=6, Use Linear Drag=true, Use Rotational Drag=false)
   &ScaleMeshSize(Scale Factor=&VectorFromFloat(Value=&FloatFromCurve(FloatCurve=DataInterface() {propertyValues {Curve {
         defaultValue=1
         keys [
          <at 0 0 RCIM_Cubic>
          {arriveTangent=1.4268, leaveTangent=1.4268, time=0.0935, value=1}
          {arriveTangent=-0.0405, interpMode=RCIM_Cubic, leaveTangent=-0.0405, tangentMode=RCTM_User, time=0.3, value=1}
          {arriveTangent=-4.4328, interpMode=RCIM_Cubic, leaveTangent=-4.4328, tangentMode=RCTM_User, time=1, value=0}
         ]
        }}})))
   &SolveForcesAndVelocity(Clamp Velocity=false, Limit Acceleration=false)
  ]
  rendererClasses [&NiagaraMeshRendererProperties]
  renderers [{rendererClass=&NiagaraMeshRendererProperties, rendererIndex=0}]
 }
 GlitchShapes {
  bEnabled=false
  emitterUpdateScript [
   &EmitterState()
   &SpawnRate(SpawnRate=12, Use Spawn Probability=false)
   &SpawnRate(SpawnGroup=1, SpawnRate=&Lerp_Float(A=8, B=17, Alpha=$System.Dive), Use Spawn Probability=false) {module=SpawnRate001}
  ]
  particleSpawnScript [
   /SetVariables(Particles.RandomBone:NiagaraInt32=&RandomRangeInt(Minimum=0, Maximum=5)) {isSetParameters=true}
   /SetVariables(Particles.BoneIndex:NiagaraInt32=&MaskIntBySpawnGroup(
     Masked Int=&RandomRangeInt(Minimum=0, Maximum=1)
     Passthrough Int=&SelectIntFromArray(Direct Array Index=$Particles.RandomBone, Int Selection Array=DataInterface() {propertyValues {IntData [0,1,5,6,9,10]}})
     Spawn Group Mask=1
    )) {isSetParameters=true}
   &InitializeParticle(
    Color=&RandomRangeLinearColor(Minimum=<rgba 0.9062 0.332 2 1>, Maximum=<rgba 0.0833 2.5136 4 1>)
    Lifetime Max=0.35
    Lifetime Min=0.1
    Lifetime Mode=::NewEnumerator1
    Mesh Scale=&Multiply_VectorByFloat(
     Vector=&RandomRangeVector(Minimum=<0.1 0.1 0.1>, Maximum=<0.4 0.4 0.4>)
     Float=&SelectFloatFromArray(Direct Array Index=$Particles.RandomBone, Float Selection Array=DataInterface() {propertyValues {FloatData [1.2,1.2,0.8,0.8,0.5,0.5]}})
    )
    Mesh Scale Mode=::NewEnumerator4
    Sprite Size Mode=::NewEnumerator3
    UsePositionOffset=false
    Write Material Random=true
   )
   &InitialMeshOrientation(Use Rotation Vector=false)
   &SkeletalMeshLocation(Bone / Socket Index=$Particles.BoneIndex, Bone Sampling Mode=::NewEnumerator4, MaskSpawnGroup=false, Orientation Sampling=::NewEnumerator1)
   &CameraOffset(Camera Offset Amount=8)
   &AddVelocity(
    Velocity=<0 0 1>
    Velocity Speed Scale=&MaskFloatBySpawnGroup(
     Masked Float=&RandomRangeFloat(Minimum=200, Maximum=300)
     Passthrough Float=&RandomRangeFloat(Minimum=0, Maximum=40)
     Spawn Group Mask=1
    )
   )
   /SetVariables(Particles.Position=&Add_Vector(
     A=$Particles.SkeletalMeshLocation.SampledPosition
     B=&RandomRangeVector(Minimum=<-5 -5 -17>, Maximum=<5 5 -5>)
    )) {isSetParameters=true}
   /SetVariables(Particles.MeshIndex=&RandomRangeInt(Minimum=0, Maximum=1)) {enabled=false, isSetParameters=true}
  ]
  particleUpdateScript [
   &ParticleState()
   /SetVariables(Particles.Jitter:NiagaraFloat=&ClampFloat(
     Float=&Waveform(
      Global Clamp Min=-1
      "Waveform [1]"=::NewEnumerator12
      "[1] Frequency"=&RandomRangeFloat(Minimum=15, Maximum=25)
      "[1] Amplitude Min/Max"=<-150 150>
     )
     Min=-0.25
     Max=0.25
    )) {isSetParameters=true}
   &DynamicMaterialParameters(
    Index 0 Param 1=&Add_Float(A=$Particles.Jitter, B=&RandomRangeFloat(Minimum=0, Maximum=1))
    Index 0 Param 2=&Subtract_Float(A=&RandomRangeFloat(Minimum=0, Maximum=1), B=$Particles.Jitter)
    Index 0 Param 3=$Particles.Jitter
    Module.Param0WriteEnabled=true
    Module.Param1WriteEnabled=true
    Module.Param2WriteEnabled=true
    Module.Param3WriteEnabled=false
   )
   &ScaleColor(
    Scale Alpha=&Add_Float(A=1, B=&ClampFloat(Float=$Particles.Jitter))
    ScaleA=true
    ScaleRGB=true
   )
   &UpdateMeshOrientation(Rotation Rate=0.2, Rotation Vector=&RandomRangeVector(Minimum=<0 0 0>, Maximum=<1 1 1>))
   &SolveForcesAndVelocity(Clamp Velocity=false, Limit Acceleration=false)
  ]
  rendererClasses [&NiagaraMeshRendererProperties]
  renderers [{rendererClass=&NiagaraMeshRendererProperties, rendererIndex=0}]
 }
 HandCrackle {
  bEnabled=false
  emitterUpdateScript [
   &EmitterState()
   &SpawnRate(SpawnRate=200, Use Spawn Probability=false)
  ]
  particleSpawnScript [
   &FN_SetSpawnGroup(Number of Spawn Groups=2)
   &InitializeParticle(
    Color=&Multiply_LinearColorByFloat(Linear Color=$User.VariantColor, Float=0.5)
    Lifetime Max=0.5
    Lifetime Min=0.3
    Lifetime Mode=::NewEnumerator1
    Sprite Size Mode=::NewEnumerator3
    Uniform Sprite Size=&Lerp_Float(
     A=3
     B=6
     Alpha=&FloatFromCurve(
      FloatCurve=DataInterface() {propertyValues {Curve {keys [
          {interpMode=RCIM_Cubic, tangentMode=RCTM_User, time=0, value=0}
          {arriveTangent=2.3987, interpMode=RCIM_Cubic, leaveTangent=2.3987, tangentMode=RCTM_User, time=0.6, value=0.55}
          {interpMode=RCIM_Cubic, tangentMode=RCTM_User, time=1, value=1}
         ]}}}
      CurveIndex=&RandomRangeFloat(Minimum=0, Maximum=1)
     )
    )
    UsePositionOffset=false
    Write Material Random=true
   )
   &SkeletalMeshLocation(
    Bone / Socket Index=&SpawnGroup()
    Bone Sampling Mode=::NewEnumerator4
    Interpolate Mesh Sampling=false
    MaskSpawnGroup=false
    Orientation Sampling=::NewEnumerator1
    Position Sampling=::NewEnumerator1
    Sampled Position Offset=&MaskVectorBySpawnGroup(Masked Vector=<8 0 0>, Passthrough Vector=<-8 0 0>)
   )
   &SphereLocation(
    MaskSpawnGroup=false
    Override Seed=false
    Sphere Radius=3
    Surface Only=true
    UseNonUniformScale=false
    UseOffset=false
    UseSphereOrientation=false
   )
   &AddVelocity(
    Module.Use Falloff Distance=false
    Override Seed=false
    Velocity Mode=::NewEnumerator1
    Velocity Origin:Vector3f=$Particles.SkeletalMeshLocation.SampledPosition
    Velocity Speed=60
   )
   &CameraOffset(Camera Offset Amount=3)
  ]
  particleUpdateScript [
   &ParticleState()
   &VectorNoiseForce(
    Force Amount=&FloatFromCurve(
     FloatCurve=DataInterface() {propertyValues {Curve {keys [{interpMode=RCIM_Cubic, tangentMode=RCTM_User, time=0, value=1}, {interpMode=RCIM_Cubic, tangentMode=RCTM_User, time=1, value=0}]}}}
     Scale Curve=1500
    )
    Module.Use Random Noise Time Interval=false
    Random Noise Time Interval=0.15
   )
   &WindForce(
    Wind Speed=&TransformVector(Vector=<0 0 1>, Source Space=::World, Destination Space=::Local)
    Wind Speed Scale=800
   )
   /Niagara/Modules/Update/Forces/LineAttractionForce(
    Attraction Strength=500
    Line End:Vector3f=&Add_Vector(A=$Particles.Initial.SkeletalMeshLocation.SampledPosition, B=<0 0 500>)
    Line Start:Vector3f=$Particles.Initial.SkeletalMeshLocation.SampledPosition
    Use Attraction Falloff=false
   )
   &Drag(Drag=7, Use Linear Drag=true, Use Rotational Drag=false)
   &ScaleSpriteSize(
    Uniform Curve Scale=&Waveform(
     Global Phase=$Particles.Age
     "Waveform [1]"=::NewEnumerator12
     "[1] Frequency"=&RandomRangeFloat(Minimum=15, Maximum=40)
     "[1] Amplitude Min/Max"=<0.5 1.5>
    )
    Uniform Curve Sprite Scale=DataInterface() {propertyValues {Curve {keys [
        {arriveTangent=10.8984, interpMode=RCIM_Cubic, leaveTangent=10.8984, tangentMode=RCTM_User, time=0, value=0}
        <at 0.2 1>
        {interpMode=RCIM_Cubic, tangentMode=RCTM_User, time=0.6, value=1}
        {arriveTangent=-5.0292, interpMode=RCIM_Cubic, leaveTangent=-5.0292, tangentMode=RCTM_User, time=1, value=0}
       ]}}}
   )
   /SetVariables(
    Particles.SpriteParams0:Vector4f=&MakeFloat4FromVectorAndFloat("Vector (RGB)"=&MakeVectorFromLinearColorRGB(Color=<rgba 0 2.4575 3 1>), "Float (Alpha)"=1)
    Particles.SpriteParams1:Vector4f=&MakeFloat4FromVectorAndFloat(
     "Vector (RGB)"=&Subtract_Vector(A=$Particles.SkeletalMeshLocation.SampledPosition, B=Linked() {linkedVariable {name=Particles.Position, type=NiagaraPosition}})
     "Float (Alpha)"=2
    )
   ) {isSetParameters=true, module=SetVariables_#4}
   &SolveForcesAndVelocity(Clamp Velocity=false, Limit Acceleration=false)
   &ScaleColor(ScaleA=true, ScaleRGB=true) {enabled=false}
  ]
  rendererClasses [&NiagaraSpriteRendererProperties]
  renderers [{rendererClass=&NiagaraSpriteRendererProperties}]
 }
 HandCrackle001 {
  bEnabled=false
  emitterUpdateScript [
   &EmitterState()
   &SpawnRate(SpawnRate=200, Use Spawn Probability=false)
  ]
  particleSpawnScript [
   &FN_SetSpawnGroup(Number of Spawn Groups=2)
   &InitializeParticle(
    Color=&Multiply_LinearColorByFloat(Linear Color=$User.VariantColor, Float=0.5)
    Lifetime Max=0.5
    Lifetime Min=0.3
    Lifetime Mode=::NewEnumerator1
    Sprite Size Mode=::NewEnumerator3
    Uniform Sprite Size=&Lerp_Float(
     A=3
     B=6
     Alpha=&FloatFromCurve(
      FloatCurve=DataInterface() {propertyValues {Curve {keys [
          {interpMode=RCIM_Cubic, tangentMode=RCTM_User, time=0, value=0}
          {arriveTangent=2.3987, interpMode=RCIM_Cubic, leaveTangent=2.3987, tangentMode=RCTM_User, time=0.6, value=0.55}
          {interpMode=RCIM_Cubic, tangentMode=RCTM_User, time=1, value=1}
         ]}}}
      CurveIndex=&RandomRangeFloat(Minimum=0, Maximum=1)
     )
    )
    UsePositionOffset=false
    Write Material Random=true
   )
   &SkeletalMeshLocation(
    Bone / Socket Index=&SpawnGroup()
    Bone Sampling Mode=::NewEnumerator4
    Interpolate Mesh Sampling=false
    MaskSpawnGroup=false
    Orientation Sampling=::NewEnumerator1
    Position Sampling=::NewEnumerator1
    Sampled Position Offset=&MaskVectorBySpawnGroup(Masked Vector=<8 0 0>, Passthrough Vector=<-8 0 0>)
   )
   &SphereLocation(
    MaskSpawnGroup=false
    Override Seed=false
    Sphere Radius=3
    Surface Only=true
    UseNonUniformScale=false
    UseOffset=false
    UseSphereOrientation=false
   )
   &AddVelocity(
    Module.Use Falloff Distance=false
    Override Seed=false
    Velocity Mode=::NewEnumerator1
    Velocity Origin:Vector3f=$Particles.SkeletalMeshLocation.SampledPosition
    Velocity Speed=60
   )
   &CameraOffset(Camera Offset Amount=3)
  ]
  particleUpdateScript [
   &ParticleState()
   &VectorNoiseForce(
    Force Amount=&FloatFromCurve(
     FloatCurve=DataInterface() {propertyValues {Curve {keys [{interpMode=RCIM_Cubic, tangentMode=RCTM_User, time=0, value=1}, {interpMode=RCIM_Cubic, tangentMode=RCTM_User, time=1, value=0}]}}}
     Scale Curve=1500
    )
    Module.Use Random Noise Time Interval=false
    Random Noise Time Interval=0.15
   )
   &WindForce(
    Wind Speed=&TransformVector(Vector=<0 0 1>, Source Space=::World, Destination Space=::Local)
    Wind Speed Scale=800
   )
   &LineAttractionForce(
    Attraction Strength=500
    Line End:Vector3f=&Add_Vector(A=$Particles.Initial.SkeletalMeshLocation.SampledPosition, B=<0 0 500>)
    Line Start:Vector3f=$Particles.Initial.SkeletalMeshLocation.SampledPosition
    Use Attraction Falloff=false
   )
   &Drag(Drag=7, Use Linear Drag=true, Use Rotational Drag=false)
   &ScaleSpriteSize(
    Uniform Curve Scale=&Waveform(
     Global Phase=$Particles.Age
     "Waveform [1]"=::NewEnumerator12
     "[1] Frequency"=&RandomRangeFloat(Minimum=15, Maximum=40)
     "[1] Amplitude Min/Max"=<0.5 1.5>
    )
    Uniform Curve Sprite Scale=DataInterface() {propertyValues {Curve {keys [
        {arriveTangent=10.8984, interpMode=RCIM_Cubic, leaveTangent=10.8984, tangentMode=RCTM_User, time=0, value=0}
        <at 0.2 1>
        {interpMode=RCIM_Cubic, tangentMode=RCTM_User, time=0.6, value=1}
        {arriveTangent=-5.0292, interpMode=RCIM_Cubic, leaveTangent=-5.0292, tangentMode=RCTM_User, time=1, value=0}
       ]}}}
   )
   /SetVariables(
    Particles.SpriteParams0:Vector4f=&MakeFloat4FromVectorAndFloat("Vector (RGB)"=&MakeVectorFromLinearColorRGB(Color=<rgba 0 2.4575 3 1>), "Float (Alpha)"=1)
    Particles.SpriteParams1:Vector4f=&MakeFloat4FromVectorAndFloat(
     "Vector (RGB)"=&Subtract_Vector(A=$Particles.SkeletalMeshLocation.SampledPosition, B=Linked() {linkedVariable {name=Particles.Position, type=NiagaraPosition}})
     "Float (Alpha)"=2
    )
   ) {isSetParameters=true, module=SetVariables_#4}
   &SolveForcesAndVelocity(Clamp Velocity=false, Limit Acceleration=false)
   &ScaleColor(ScaleA=true, ScaleRGB=true) {enabled=false}
  ]
  rendererClasses [&NiagaraSpriteRendererProperties]
  renderers [{rendererClass=&NiagaraSpriteRendererProperties}]
 }
 HandCrackle002 {
  emitterUpdateScript [
   &EmitterState()
   &SpawnRate(SpawnRate=300, Use Spawn Probability=false)
  ]
  particleSpawnScript [
   &FN_SetSpawnGroup(Number of Spawn Groups=2)
   &InitializeParticle(
    Color=&Multiply_LinearColorByFloat(Linear Color=$User.VariantColor, Float=2)
    Lifetime Max=0.75
    Lifetime Min=0.1
    Lifetime Mode=::NewEnumerator1
    Mesh Renderer Array Visibility Mode=::NewEnumerator3
    Mesh Renderer Info=DataInterface() {propertyValues {MeshRenderer=&self:HandCrackle002_0.NiagaraMeshRendererProperties_0}}
    Mesh Scale Mode=::NewEnumerator3
    Mesh Uniform Scale=&Multiply_Float(
     A=&RemapRange(
      Input Value To Remap=&FloatFromCurve(
       FloatCurve=DataInterface() {propertyValues {Curve {
          defaultValue=0.05
          keys [
           {tangentMode=RCTM_User, time=0, value=0}
           {
            arriveTangent=0.0405
            arriveTangentWeight=0.2338
            interpMode=RCIM_Cubic
            leaveTangent=0.0405
            leaveTangentWeight=0.0998
            tangentMode=RCTM_User
            time=0.77
            value=0
           }
           {arriveTangent=12.0972, arriveTangentWeight=0.8048, interpMode=RCIM_Cubic, leaveTangent=12.0972, tangentMode=RCTM_User, time=1, value=1}
          ]
         }}}
       CurveIndex=&RandomRangeFloat(Minimum=0, Maximum=1)
      )
      Input Min=0
      Input Max=1
      Output Min=0.05
      Output Max=0.15
     )
     B=&FN_FrontEndFloatModifier(Gameplay=1.5, Front End=1.5)
    )
    UsePositionOffset=false
    Write Material Random=true
   )
   /SetVariables(Particles.Lifetime=&Multiply_Float(
     A=$Particles.Lifetime
     B=&RemapRange(
      Input Value To Remap=&VectorLength(Vector=$Particles.Scale)
      Input Min=&FN_FrontEndFloatModifier(Gameplay=0.07, Front End=0.15)
      Input Max=&FN_FrontEndFloatModifier(Gameplay=0.18, Front End=0.25)
      Output Min=1.22
      Output Max=0.35
     )
    )) {isSetParameters=true}
   &MeshRotationForce(Rotation=&RandomRangeVector(Minimum=<5 5 5>, Maximum=<-5 -5 -5>))
   &SkeletalMeshLocation(
    Bone / Socket Index=&SpawnGroup()
    Bone Sampling Mode=::NewEnumerator4
    Interpolate Mesh Sampling=false
    MaskSpawnGroup=false
    Orientation Sampling=::NewEnumerator1
    Position Sampling=::NewEnumerator1
    Sampled Position Offset=&MaskVectorBySpawnGroup(Masked Vector=<8 0 0>, Passthrough Vector=<-8 0 0>)
    Skeletal Mesh=$System.Skeletal Mesh Reader
   )
   &SphereLocation(
    MaskSpawnGroup=false
    Override Seed=false
    Sphere Radius=3
    Surface Only=true
    UseNonUniformScale=false
    UseOffset=false
    UseSphereOrientation=false
   )
   &AddVelocity(
    Module.Use Falloff Distance=false
    Override Seed=false
    Velocity Mode=::NewEnumerator1
    Velocity Origin:Vector3f=$Particles.SkeletalMeshLocation.SampledPosition
    Velocity Speed=60
   )
   &ApplyInitialForces()
   &CameraOffset(Camera Offset Amount=3)
  ]
  particleUpdateScript [
   &ParticleState()
   &VectorNoiseForce(
    Force Amount=&FloatFromCurve(
     FloatCurve=DataInterface() {propertyValues {Curve {keys [{interpMode=RCIM_Cubic, tangentMode=RCTM_User, time=0, value=1}, {interpMode=RCIM_Cubic, tangentMode=RCTM_User, time=1, value=0}]}}}
     Scale Curve=1500
    )
    Module.Use Random Noise Time Interval=false
    Random Noise Time Interval=0.15
   )
   &WindForce(
    Wind Speed=&TransformVector(Vector=<0 0 1>, Source Space=::World, Destination Space=::Local)
    Wind Speed Scale=800
   )
   &LineAttractionForce(
    Attraction Strength=500
    Line End:Vector3f=&Add_Vector(A=$Particles.Initial.SkeletalMeshLocation.SampledPosition, B=<0 0 500>)
    Line Start:Vector3f=$Particles.Initial.SkeletalMeshLocation.SampledPosition
    Use Attraction Falloff=false
   )
   &Drag(
    Drag=&Multiply_Float(
     A=&FN_FrontEndFloatModifier(Gameplay=6, Front End=7)
     B=&RemapRange(
      Input Value To Remap=&VectorLength(Vector=$Particles.Scale)
      Input Min=0.05
      Input Max=1
      Output Min=1
      Output Max=1.5
     )
    )
    Use Linear Drag=true
    Use Rotational Drag=false
   )
   &ScaleMeshSize(Scale Factor=&VectorFromFloat(Value=&FloatFromCurve(FloatCurve=DataInterface() {propertyValues {Curve {
         defaultValue=1
         keys [
          <at 0 0 RCIM_Cubic>
          {arriveTangent=1.4268, leaveTangent=1.4268, time=0.0935, value=1}
          {arriveTangent=-0.0405, interpMode=RCIM_Cubic, leaveTangent=-0.0405, tangentMode=RCTM_User, time=0.7009, value=1}
          {arriveTangent=-8.0089, interpMode=RCIM_Cubic, leaveTangent=-8.0089, tangentMode=RCTM_User, time=1, value=0}
         ]
        }}})))
   &SolveForcesAndVelocity(Clamp Velocity=false, Limit Acceleration=false)
  ]
  rendererClasses [&NiagaraMeshRendererProperties]
  renderers [{rendererClass=&NiagaraMeshRendererProperties, rendererIndex=0}]
 }
 HandCrackle003 {
  emitterUpdateScript [
   &EmitterState()
   &SpawnRate(SpawnRate=240, Use Spawn Probability=false)
  ]
  particleSpawnScript [
   &FN_SetSpawnGroup(Number of Spawn Groups=2)
   &InitializeParticle(
    Color=&Multiply_LinearColorByFloat(Linear Color=$User.VariantColor, Float=2)
    Lifetime Max=0.75
    Lifetime Min=0.1
    Lifetime Mode=::NewEnumerator1
    Mesh Renderer Array Visibility Mode=::NewEnumerator3
    Mesh Renderer Info=DataInterface() {propertyValues {MeshRenderer=&self:HandCrackle003_1.NiagaraMeshRendererProperties_0}}
    Mesh Scale Mode=::NewEnumerator3
    Mesh Uniform Scale=&Multiply_Float(
     A=&RemapRange(
      Input Value To Remap=&FloatFromCurve(
       FloatCurve=DataInterface() {propertyValues {Curve {
          defaultValue=0.05
          keys [
           {tangentMode=RCTM_User, time=0, value=0}
           {
            arriveTangent=0.0405
            arriveTangentWeight=0.2338
            interpMode=RCIM_Cubic
            leaveTangent=0.0405
            leaveTangentWeight=0.0998
            tangentMode=RCTM_User
            time=0.77
            value=0
           }
           {arriveTangent=12.0972, arriveTangentWeight=0.8048, interpMode=RCIM_Cubic, leaveTangent=12.0972, tangentMode=RCTM_User, time=1, value=1}
          ]
         }}}
       CurveIndex=&RandomRangeFloat(Minimum=0, Maximum=1)
      )
      Input Min=0
      Input Max=1
      Output Min=&FN_FrontEndFloatModifier(Gameplay=0.0375, Front End=0.05)
      Output Max=0.15
     )
     B=&FN_FrontEndFloatModifier(Gameplay=1.5, Front End=1.5)
    )
    UsePositionOffset=false
    Write Material Random=true
   )
   /SetVariables(Particles.Lifetime=&Multiply_Float(
     A=$Particles.Lifetime
     B=&RemapRange(
      Input Value To Remap=&VectorLength(Vector=$Particles.Scale)
      Input Min=&FN_FrontEndFloatModifier(Gameplay=0.07, Front End=0.15)
      Input Max=&FN_FrontEndFloatModifier(Gameplay=0.18, Front End=0.25)
      Output Min=1.22
      Output Max=0.15
     )
    )) {isSetParameters=true}
   &MeshRotationForce(Rotation=&RandomRangeVector(Minimum=<5 5 5>, Maximum=<-5 -5 -5>))
   &SkeletalMeshLocation(
    Bone / Socket Index=&MergeHelper_IntSwitch(0=2, 1=3, 2=?, 3=?, 4=?, 5=?, Index=&SpawnGroup(), Num=::NewEnumerator0)
    Bone Sampling Mode=::NewEnumerator4
    Interpolate Mesh Sampling=false
    MaskSpawnGroup=false
    Orientation Sampling=::NewEnumerator1
    Position Sampling=::NewEnumerator1
    Sampled Position Offset=&MaskVectorBySpawnGroup(Masked Vector=<-12 5 0>, Passthrough Vector=<12 -5 0>)
    Skeletal Mesh=$System.Skeletal Mesh Reader
   )
   &SphereLocation(
    MaskSpawnGroup=false
    Override Seed=false
    Sphere Radius=3
    Surface Only=true
    UseNonUniformScale=false
    UseOffset=false
    UseSphereOrientation=false
   )
   &AddVelocity(
    Module.Use Falloff Distance=false
    Override Seed=false
    Velocity Mode=::NewEnumerator1
    Velocity Origin:Vector3f=$Particles.SkeletalMeshLocation.SampledPosition
    Velocity Speed=60
   )
   &ApplyInitialForces()
   &CameraOffset(Camera Offset Amount=3)
  ]
  particleUpdateScript [
   &ParticleState()
   &VectorNoiseForce(
    Force Amount=&FloatFromCurve(
     FloatCurve=DataInterface() {propertyValues {Curve {keys [{interpMode=RCIM_Cubic, tangentMode=RCTM_User, time=0, value=1}, {interpMode=RCIM_Cubic, tangentMode=RCTM_User, time=1, value=0}]}}}
     Scale Curve=1500
    )
    Module.Use Random Noise Time Interval=false
    Random Noise Time Interval=0.15
   )
   &WindForce(
    Wind Speed=&TransformVector(Vector=<0 0 1>, Source Space=::World, Destination Space=::Local)
    Wind Speed Scale=800
   )
   &LineAttractionForce(
    Attraction Strength=500
    Line End:Vector3f=&Add_Vector(A=$Particles.Initial.SkeletalMeshLocation.SampledPosition, B=<0 0 500>)
    Line Start:Vector3f=$Particles.Initial.SkeletalMeshLocation.SampledPosition
    Use Attraction Falloff=false
   )
   &Drag(
    Drag=&Multiply_Float(
     A=&FN_FrontEndFloatModifier(Gameplay=6, Front End=7)
     B=&RemapRange(
      Input Value To Remap=&VectorLength(Vector=$Particles.Scale)
      Input Min=0.05
      Input Max=1
      Output Min=1
      Output Max=1.5
     )
    )
    Use Linear Drag=true
    Use Rotational Drag=false
   )
   &ScaleMeshSize(Scale Factor=&VectorFromFloat(Value=&FloatFromCurve(FloatCurve=DataInterface() {propertyValues {Curve {
         defaultValue=1
         keys [
          <at 0 0 RCIM_Cubic>
          {arriveTangent=1.4268, leaveTangent=1.4268, time=0.0935, value=1}
          {arriveTangent=-0.0405, interpMode=RCIM_Cubic, leaveTangent=-0.0405, tangentMode=RCTM_User, time=0.7009, value=1}
          {arriveTangent=-8.0089, interpMode=RCIM_Cubic, leaveTangent=-8.0089, tangentMode=RCTM_User, time=1, value=0}
         ]
        }}})))
   &SolveForcesAndVelocity(Clamp Velocity=false, Limit Acceleration=false)
  ]
  rendererClasses [&NiagaraMeshRendererProperties]
  renderers [{rendererClass=&NiagaraMeshRendererProperties, rendererIndex=0}]
 }
 Lights {
  emitterSpawnScript [&FN_PreserveSkydiveUserVariables() {module=PreserveUserVariables}]
  emitterUpdateScript [
   &EmitterState(Inactive Response=::NewEnumerator1, Life Cycle Mode=::NewEnumerator1, Loop Duration=0, UseLoopDelay=false)
   &SpawnBurst_Instantaneous(Spawn Count=4, Spawn Probability=HlslExpression() {hlslExpression="Emitter.LoopCount == 0"}, Use Spawn Probability=true) {module=SpawnBurst_Instantaneous001}
  ]
  particleSpawnScript [
   &FN_SetSpawnGroup()
   &InitializeParticle(
    Color Mode=::NewEnumerator0
    Lifetime=0
    Sprite Size Mode=::NewEnumerator3
    Uniform Sprite Size=50
    UsePositionOffset=false
    Write Material Random=true
   ) {module=InitializeParticle001}
   &CameraOffset(Camera Offset Amount=11)
   &Color(Color=&Multiply_LinearColorByFloat(Linear Color=$User.VariantColor, Float=1.5))
   &DynamicMaterialParameters(
    Index 0 Param 1=&MaskFloatBySpawnGroup(Masked Float=0.5, Passthrough Float=0)
    Index 0 Param 2=&FN_FrontEndFloatModifier(Gameplay=0, Front End=1)
    Module.Param0WriteEnabled=true
    Module.Param1WriteEnabled=true
    Module.Param2WriteEnabled=false
    Module.Param3WriteEnabled=false
   )
  ]
  particleUpdateScript [
   &ParticleState(Kill Particles When Lifetime Has Elapsed=false)
   &SampleSkeletalMeshSkeleton(
    Mesh:NiagaraDataInterfaceSkeletalMesh=$System.Skeletal Mesh Reader
    Sampled Rotation Coordinate Space:ENiagaraCoordinateSpace=::Local
    Specific Bone Index:NiagaraInt32=$Particles.SpawnGroup
   )
   &SkeletalMeshLocation(
    Apply Sampled Orientation to Mesh Particles=true
    Bone / Socket Index=$Particles.SpawnGroup
    Bone Sampling Mode=::NewEnumerator4
    Inherit Velocity=false
    Interpolate Mesh Sampling=false
    MaskSpawnGroup=false
    Sampled Position Offset=&MergeHelper_VectorSwitch(0=<15 0 0>, 1=<-15 0 0>, 2=<0 15 0>, 3=<0 -15 0>, Index=$Particles.SpawnGroup, Num=::NewEnumerator2)
    Sampled Position Offset Coordinate Space:ENiagaraMeshTransforms=::NewEnumerator3
    Skeletal Mesh=$System.Skeletal Mesh Reader
   )
  ]
  rendererClasses [&NiagaraSpriteRendererProperties]
  renderers [{rendererClass=&NiagaraSpriteRendererProperties}]
 }
 Lights001 {
  emitterSpawnScript [&FN_PreserveSkydiveUserVariables() {module=PreserveUserVariables}]
  emitterUpdateScript [
   &EmitterState(Inactive Response=::NewEnumerator1, Life Cycle Mode=::NewEnumerator1, Loop Duration=0, UseLoopDelay=false)
   &SpawnBurst_Instantaneous(Spawn Count=4, Spawn Probability=HlslExpression() {hlslExpression="Emitter.LoopCount == 0"}, Use Spawn Probability=true) {module=SpawnBurst_Instantaneous001}
  ]
  particleSpawnScript [
   &FN_SetSpawnGroup()
   &InitializeParticle(
    Color Mode=::NewEnumerator0
    Lifetime=0
    Sprite Size Mode=::NewEnumerator3
    Uniform Sprite Size=50
    UsePositionOffset=false
    Write Material Random=true
   ) {module=InitializeParticle001}
   &CameraOffset(Camera Offset Amount=11)
   &Color(Color=&Multiply_LinearColorByFloat(Linear Color=$User.VariantColor, Float=5))
   &DynamicMaterialParameters(
    Index 0 Param 1=&MaskFloatBySpawnGroup(Masked Float=0.5, Passthrough Float=0)
    Index 0 Param 2=&FN_FrontEndFloatModifier(Gameplay=0, Front End=1)
    Module.Param0WriteEnabled=true
    Module.Param1WriteEnabled=true
    Module.Param2WriteEnabled=false
    Module.Param3WriteEnabled=false
   )
  ]
  particleUpdateScript [
   &ParticleState(Kill Particles When Lifetime Has Elapsed=false)
   &SampleSkeletalMeshSkeleton(
    Mesh:NiagaraDataInterfaceSkeletalMesh=$System.Skeletal Mesh Reader
    Sampled Rotation Coordinate Space:ENiagaraCoordinateSpace=::Local
    Specific Bone Index:NiagaraInt32=$Particles.SpawnGroup
   )
   &SkeletalMeshLocation(
    Apply Sampled Orientation to Mesh Particles=true
    Bone / Socket Index=&SpawnGroup()
    Bone Sampling Mode=::NewEnumerator4
    Inherit Velocity=false
    Interpolate Mesh Sampling=false
    MaskSpawnGroup=false
    Sampled Position Offset Coordinate Space:ENiagaraMeshTransforms=::NewEnumerator3
    Skeletal Mesh=$System.Skeletal Mesh Reader
   )
   /SetVariables(Particles.Decal DP:Vector4f=&MakeVector4(
     X=&FloatFromCurve(
      FloatCurve=DataInterface() {propertyValues {Curve {keys [<at 0 0 RCIM_Cubic>,<at 0.3 1 RCIM_Cubic>]}}}
      CurveIndex=$Emitter.Age
      Scale Curve=2
     )
     Y=0.01
     Z=&FloatFromCurve(FloatCurve=DataInterface() {propertyValues {Curve {keys [<at 0 0.05>,<at 1 0>]}}})
     W=&RandomRangeFloat(Minimum=0, Maximum=1)
    )) {isSetParameters=true}
  ]
  rendererClasses [&NiagaraSpriteRendererProperties]
  renderers [{rendererClass=&NiagaraSpriteRendererProperties}]
 }
 Lights002 {
  bEnabled=false
  emitterSpawnScript [&FN_PreserveSkydiveUserVariables() {module=PreserveUserVariables}]
  emitterUpdateScript [
   &EmitterState(Inactive Response=::NewEnumerator1, Life Cycle Mode=::NewEnumerator1, Loop Duration=1, UseLoopDelay=false)
   &SpawnRate(SpawnRate=5, Use Spawn Probability=false)
  ]
  particleSpawnScript [
   &InitializeParticle(Color Mode=::NewEnumerator0, Sprite Size Mode=::NewEnumerator3, Uniform Sprite Size=50, UsePositionOffset=false, Write Material Random=true) {module=InitializeParticle001}
   &Color(Color=&Multiply_LinearColorByFloat(Linear Color=$User.VariantColor, Float=1.5))
   &SkeletalMeshLocation(
    Apply Sampled Orientation to Mesh Particles=true
    Bone Sampling Mode=::NewEnumerator0
    Inherit Velocity=false
    Interpolate Mesh Sampling=false
    MaskSpawnGroup=false
    Module.Override Randomness=false
    Sampled Position Offset Coordinate Space:ENiagaraMeshTransforms=::NewEnumerator3
   )
   &DynamicMaterialParameters(
    Index 0 Param 1=&MaskFloatBySpawnGroup(Masked Float=0.5, Passthrough Float=0)
    Index 0 Param 2=&FN_FrontEndFloatModifier(Gameplay=0, Front End=1)
    Module.Param0WriteEnabled=true
    Module.Param1WriteEnabled=true
    Module.Param2WriteEnabled=false
    Module.Param3WriteEnabled=false
   )
  ]
  particleUpdateScript [
   &ParticleState()
   &ScaleColor(
    Scale Alpha=&FloatFromCurve(FloatCurve=DataInterface() {propertyValues {Curve {
        defaultValue=1
        keys [
         <at 0 0 RCIM_Cubic>
         <at 0.1175 1>
         {arriveTangent=-0.0405, interpMode=RCIM_Cubic, leaveTangent=-0.0405, tangentMode=RCTM_User, time=0.7009, value=1}
         {arriveTangent=-8.0089, interpMode=RCIM_Cubic, leaveTangent=-8.0089, tangentMode=RCTM_User, time=1, value=0}
        ]
       }}})
    ScaleA=true
    ScaleRGB=true
   )
  ]
  rendererClasses [&NiagaraSpriteRendererProperties]
  renderers [{rendererClass=&NiagaraSpriteRendererProperties}]
 }
 MainRibbon {
  bEnabled=false
  emitterUpdateScript [
   /SetVariables(Emitter.EmitterRandom:NiagaraFloat=&RandomRangeFloat(Minimum=0, Maximum=1)) {isSetParameters=true}
   &EmitterState()
   &FN_SpawnRibbon_v2(Number of Ribbons=2, Spawn Mode=::NewEnumerator0, SpawnRate=15)
  ]
  particleSpawnScript [
   &FN_InitializeRibbon_v3(
    Enable Skeletal Position=true
    Lifetime=0.65
    Ribbon Width=17
    Sampled Position Coordinate Space:ENiagaraCoordinateSpace=::Local
    Sampled Rotation Coordinate Space:ENiagaraCoordinateSpace=::Local
    Sampled Velocity Coordinate Space:ENiagaraCoordinateSpace=::Local
   ) {module=FN_InitializeRibbon_v3001}
   /SetVariables(Particles.RandomWave:NiagaraFloat=&Waveform(
     Global Phase=&Add_Float(A=$System.Age, B=&MaskFloatBySpawnGroup(Masked Float=0, Passthrough Float=0.35))
     "Waveform [1]"=::NewEnumerator12
     "[1] Frequency"=5
     "[1] Amplitude Min/Max"=<0 1>
    )) {isSetParameters=true}
   &AddVelocity(Velocity=&Lerp_Vector(A=<0 0 240>, B=<1 1 400>, Alpha=$System.FN_Integrator001.Value))
   &AddVelocity(Velocity=&MakeVector(
     X=&Waveform(
      Global Phase=&Add_Float(
       A=$System.Age
       B=&MaskFloatBySpawnGroup(Masked Float=0, Passthrough Float=&Lerp_Float(A=0.4, B=0.2, Alpha=$Emitter.EmitterRandom))
      )
      Global Amplitude Scale=&Lerp_Float(A=50, B=150, Alpha=$Particles.RandomWave)
      "Waveform [1]"=::NewEnumerator0
      "[1] Frequency"=&Lerp_Float(A=0.7, B=1.5, Alpha=$Emitter.EmitterRandom)
      "[1] Amplitude Min/Max"=<-1 1>
     )
     Y=0
     Z=0
    )) {module=AddVelocity001}
   &SampleSkeletalMeshSkeleton() {enabled=false}
   &SkeletalMeshLocation(
    Bone / Socket Index=&SpawnGroup()
    Bone Sampling Mode=::NewEnumerator4
    Interpolate Mesh Sampling=false
    MaskSpawnGroup=false
    Orientation Sampling=::NewEnumerator1
    Position Sampling=::NewEnumerator1
    Sampled Position Offset=&MaskVectorBySpawnGroup(Masked Vector=<8 0 0>, Passthrough Vector=<-8 0 0>)
   )
  ]
  particleUpdateScript [
   &WindForce(Air Resistance=0.05, Wind Speed=<0 0 1>, Wind Speed Scale=1500)
   &Drag(Drag=5, Use Linear Drag=true, Use Rotational Drag=false)
   &LineAttractionForce(
    Attraction Strength:NiagaraFloat=&FloatFromCurve(
     FloatCurve=DataInterface() {propertyValues {Curve {keys [{arriveTangent=2.373, interpMode=RCIM_Cubic, time=0, value=0}, {interpMode=RCIM_Cubic, tangentMode=RCTM_User, time=1, value=1}]}}}
     Scale Curve=500
    )
    Line End:Vector3f=&Add_Vector(A=$Particles.SampledPosition, B=<0 0 500>)
    Line Start:Vector3f=$Particles.SampledPosition
    Use Attraction Falloff=false
   )
   &SolveForcesAndVelocity(Clamp Velocity=false, Limit Acceleration=false)
   &FN_UpdateRibbon(
    Enable Skeletal Position=true
    Initial Color:LinearColor=$Particles.Initial.Color
    Initial RibbonWidth:NiagaraFloat=$Particles.Initial.RibbonWidth
   )
   &Material(Base Texture Time:NiagaraFloat=&Multiply_Float(A=$Emitter.Age, B=-1), Jitter Amount=0)
   /SetVariables(Particles.Alpha:NiagaraFloat=&FloatFromCurve(
     FloatCurve=DataInterface() {propertyValues {Curve {keys [
         {arriveTangent=4.667, interpMode=RCIM_Cubic, leaveTangent=4.667, tangentMode=RCTM_User, time=0, value=0}
         {interpMode=RCIM_Cubic, tangentMode=RCTM_User, time=0.25, value=1}
         {tangentMode=RCTM_User, time=1, value=0}
        ]}}}
     Scale Curve=&FloatFromCurve(
      FloatCurve=DataInterface() {propertyValues {Curve {keys [{interpMode=RCIM_Cubic, tangentMode=RCTM_User, time=0, value=0}, {interpMode=RCIM_Cubic, tangentMode=RCTM_User, time=0.8, value=1}]}}}
      CurveIndex=&Waveform(
       Global Input Position=$Particles.NormalizedAge
       Global Phase=&Add_Float(
        A=&Multiply_Float(A=$System.Age, B=&Lerp_Float(A=-4, B=-6, Alpha=$Emitter.EmitterRandom))
        B=&MaskFloatBySpawnGroup(Masked Float=0, Passthrough Float=&Lerp_Float(A=0.5, B=2, Alpha=$Emitter.EmitterRandom))
       )
       "Waveform [1]"=::NewEnumerator0
       "[1] Frequency"=0.2
       "[1] Amplitude Min/Max"=<0 1>
      )
     )
    )) {isSetParameters=true}
   &ScaleColor(
    Scale Alpha=$Particles.Alpha
    Scale RGB=&MakeVectorFromLinearColorRGB(Color=&LerpLinearColors(StartColor=<rgba 0 0.4227 0.7 1>, EndColor=<rgba 0 0.5402 0.6 1>, LerpFactor=$Particles.Alpha))
    ScaleA=true
    ScaleRGB=true
   )
  ]
  rendererClasses [&NiagaraRibbonRendererProperties]
  renderers [{rendererClass=&NiagaraRibbonRendererProperties}]
 }
}
summary {
 systemName=NS_Contrail_Dive_RaveVision
 emitters [
  {bEnabled=false, emitterName=MainRibbon, rendererClasses [&NiagaraRibbonRendererProperties]}
  {bEnabled=false, emitterName=HandCrackle, rendererClasses [&NiagaraSpriteRendererProperties]}
  {bEnabled=false, emitterName=BodyCrackle, rendererClasses [&NiagaraSpriteRendererProperties]}
  {bEnabled=false, emitterName=GlitchShapes, rendererClasses [&NiagaraMeshRendererProperties]}
  {emitterName=Lights, rendererClasses [&NiagaraSpriteRendererProperties]}
  {emitterName=Lights001, rendererClasses [&NiagaraSpriteRendererProperties]}
  {emitterName=BodyCrackle001, rendererClasses [&NiagaraMeshRendererProperties]}
  {bEnabled=false, emitterName=HandCrackle001, rendererClasses [&NiagaraSpriteRendererProperties]}
  {bEnabled=false, emitterName=BodyCrackle002, rendererClasses [&NiagaraSpriteRendererProperties]}
  {emitterName=HandCrackle002, rendererClasses [&NiagaraMeshRendererProperties]}
  {emitterName=BodyCrackle003, rendererClasses [&NiagaraMeshRendererProperties]}
  {emitterName=HandCrackle003, rendererClasses [&NiagaraMeshRendererProperties]}
  {bEnabled=false, emitterName=Lights002, rendererClasses [&NiagaraSpriteRendererProperties]}
  {bEnabled=false, emitterName=BodyCrackle004, rendererClasses [&NiagaraMeshRendererProperties]}
  {emitterName=BodyCrackle005, rendererClasses [&NiagaraMeshRendererProperties]}
 ]
 userVariables [
  User.bIsFrontEnd {type=NiagaraBool}
  User.bIsFrontEndPreview {type=NiagaraBool}
  User.PlayerSpeed {type=NiagaraFloat, defaultValue=0, description=""}
  User.RibbonIDOffset {type=NiagaraInt32}
  User.VariantColor {type=LinearColor, description="", defaultValue {struct=LinearColor, value=<rgba 1 0 0 1>}}
 ]
}
